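; AI evaluator for the CasinoReaper NPC that decides whether it should start leaving its vehicle.
; On enter, it caches the controlled pawn from the behavior tree component's owning controller.
; The starting condition passes if the blackboard vehicle actor has any disabled tag, or if the pawn has a tag allowing it to leave.
; Otherwise it returns false; the vehicle cast in the pawn-tag branch has no effect on the result.
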(event EventOnEnter
  (params (OwnerComp :type Behavior%20Tree%20Component%20Object%20Reference))
  (bind _ascontroller (Utilities|Casting|CastToController (Components|GetOwner OwnerComp)))
  (Variables|Default|SetPawn (Pawn|GetControlledPawn _ascontroller)))

(fn EvaluateStartingConditions
  (params (OwnerComp :type Behavior%20Tree%20Component%20Object%20Reference))
  (Utilities|FlowControl|Sequence
    (:then_0
      ;[1 1 1] Vehicle is disabled
      (bind _asactor (Utilities|Casting|CastToActor (Blackboard|GetObject (Variables|Default|GetVehicle%20Actor) OwnerComp)))
      (bind _returnvalue (GameplayTags|HasAnyMatchingGameplayTags(Message) _asactor (Variables|Default|GetVehicle%20Disabled%20Tags)))
      (if _returnvalue
        (return true)))
    (:then_1
      ; AI Should no longer be able to use Vehicle because it has a GameplayTag on it.
      (bind _returnvalue_2 (GameplayTags|HasAnyMatchingGameplayTags(Message) (Variables|Default|GetPawn) (Variables|Default|GetTagToAllowLeavingVehicle)))
      (if _returnvalue_2
        (Utilities|Casting|CastToActor (Blackboard|GetObject (Variables|Default|GetVehicle%20Actor) OwnerComp))
        (return true)))
    (:then_2
      ;[1 1 1] Don't leave vehicle
      (return false))))
