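# BT_InVehicle is the CasinoReaper NPC's behavior tree for the in-vehicle state, using the shared humanoid BB_InVehicle blackboard.
# A root Sequence runs a single Wait task of five seconds, with no decorators or other logic, so the tree only idles.
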BehaviorTree /NPCLibraryCh6/NPCs/CasinoReaper/BT_InVehicle
blackboard=/AICore/Humanoids/BehaviorTrees/Blackboards/BB_InVehicle
root {
 name=Sequence
 class=BTComposite_Sequence
 children [Wait {class=BTTask_Wait, properties {waitTime {defaultValue=5}}}]
 decorators []
 properties {children [{childTask=&self:BTTask_Wait_0}]}
}
