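; Valet sports car Blueprint for the Casino Reaper NPC content that reacts to who enters and leaves the vehicle.
; A human pawn entering seat 0 receives a loose gameplay tag from the GetTagForSDACheck variable, and any human exit removes it.
; A bot entering seat 1 plays a speech bubble with voice-over, stopping its previous line.
; BeginPlay and construction only call the parent, and the remaining events are empty.
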(event EventPickRandomMaterial)

(event EventBeginPlay
  (|CallParentFunction))

(event Collision|EventActorBeginOverlap
  (params (OtherActor :type Actor%20Object%20Reference)))

(event EventTick
  (params (DeltaSeconds :type Float%20%28single-precision%29)))

(event EventPawnEnterSeat
  (params (Vehicle :type Fort%20Vehicle%20Interface%20Interface%20%28by%20ref%29) (PlayerPawn :type Fort%20Pawn%20Object%20Reference) (SeatIndex :type Integer))
  (if (and (not (Pawn|IsBotControlled PlayerPawn)) (== SeatIndex 0))
    (Ability|GameplayEffect|AddLooseGameplayTags PlayerPawn (Variables|Default|GetTagForSDACheck) true)
    (elif (and (Pawn|IsBotControlled PlayerPawn) (== SeatIndex 1))
      (Class|BPFLAICommonFunctions|PlaySpeechBubble :Pawn PlayerPawn :StopPreviousLine true :TryPlayVOLine true))))

(event EventOnPawnExitVehicle
  (params (PlayerPawn :type Fort%20Player%20Pawn%20Object%20Reference) (ExitSocketName :type Name :default "None"))
  (if (not (Pawn|IsBotControlled PlayerPawn))
    (Ability|GameplayEffect|RemoveLooseGameplayTags PlayerPawn (Variables|Default|GetTagForSDACheck) true)))

(fn ConstructionScript
  (|CallParentFunction))
